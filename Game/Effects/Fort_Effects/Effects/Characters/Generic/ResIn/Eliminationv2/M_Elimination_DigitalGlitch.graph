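# Translucent unlit Niagara sprite material for the elimination effect, drawing a digitally glitched mask with RGB channel separation.
# Panned digital noise, driven by dynamic Speed and Random Value, distorts UVs and yields a transformed vector scaled by GlitchAmount.
# Chroma spreads three offset side-mask samples, multiplied with axis-swapped copies to mask both axes.
# Particle color tints the HDR-remapped emissive; the ColorA/ColorB blend and scalar GlitchAmount feed no visible node.
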Material /Game/Effects/Fort_Effects/Effects/Characters/Generic/ResIn/Eliminationv2/M_Elimination_DigitalGlitch
graph {
 properties {bUsedWithNiagaraSprites=true, blendMode=BLEND_Translucent, pixelDepthOffsetMode=PDOM_Legacy, shadingModel=MSM_Unlit, usedShadingModels=Unlit}
 nodes [
  LinearInterpolate(
   A=VectorParameter() {parameterName=ColorA, defaultValue=<rgba 0.1912 0.5906 0.6038 1>}.RGB
   B=VectorParameter() {parameterName=ColorB, defaultValue=<rgba 0.6333 1.9563 2 1>}.RGB
   Alpha=DynamicParameter() {defaultValue=<rgba 0 1 1 1>, paramNames [Random Value, Speed, GlitchAmount, Chroma]}.Random Value
  )
  n1 = TextureCoordinate() {uTiling=0.01, vTiling=0.01}
  n2 = DynamicParameter() {defaultValue=<rgba 0 1 1 1>, paramNames [Random Value, Speed, GlitchAmount, Chroma]}
  n3 = Time()
  n4 = TextureSample(UVs=Panner(Coordinate=Add(A=@n1, B=@n2.Random Value), Time=Multiply(A=@n2.Speed, B=@n3)) {speedX=0.1, speedY=0.1}) {samplerType=SAMPLERTYPE_LinearColor, texture=/Game/Global/Textures/Default/Noise/T_Atlas_DigitalNoise_01}
  n5 = Add(A=TextureCoordinate(), B=Multiply(A=ComponentMask(None=@n4.RGB) {g=true, r=true}) {constB=0.025})
  n6 = LinearInterpolate(
   A=Constant2Vector()
   B=ComponentMask(None=VectorParameter() {parameterName=ChromaOffset, defaultValue=<rgba 0.1 0.1 0 0>}.RGB) {g=true, r=true}
   Alpha=DynamicParameter() {defaultValue=<rgba 0 1 1 1>, paramNames [Random Value, Speed, GlitchAmount, Chroma]}.Chroma
  )
  n7 = Add(A=@n5, B=@n6)
  n8 = Add(A=@n5, B=Multiply(A=@n6) {constB=-1})
  n9 = MaterialFunctionCall(XY=@n7) {materialFunction=&Swizzle}
  n10 = MaterialFunctionCall(XY=@n5) {materialFunction=&Swizzle}
  n11 = MaterialFunctionCall(XY=@n8) {materialFunction=&Swizzle}
  n12 = Multiply(
   A=MaterialFunctionCall(
    X=TextureSample(UVs=@n7) {texture=/Game/Effects/Fort_Effects/Textures/Masks/T_Side_Mask}.G
    Y=TextureSample(UVs=@n5) {texture=&T_Side_Mask}.G
    Z=TextureSample(UVs=@n8) {texture=&T_Side_Mask}.G
   ) {materialFunction=&MakeFloat3}.Result
   B=MaterialFunctionCall(
    X=TextureSample(UVs=@n9.YX) {texture=&T_Side_Mask}.G
    Y=TextureSample(UVs=@n10.YX) {texture=&T_Side_Mask}.G
    Z=TextureSample(UVs=@n11.YX) {texture=&T_Side_Mask}.G
   ) {materialFunction=&MakeFloat3}.Result
  )
  n13 = ParticleColor()
  n14 = Multiply(A=@n12, B=@n13.RGB)
  MaterialFunctionCall(Emissive=MaterialFunctionCall(LDR Input=Saturate(None=@n14), HDR Tint=@n14) {materialFunction=&DeriveHDRfromLDR}.RemappedRange) {materialFunction=&MF_FX_EmissiveAdjustments}
  Multiply(A=@n13.A, B=@n12)
  Multiply(
   A=Transform(None=@n4.RGB)
   B=DynamicParameter() {defaultValue=<rgba 0 1 1 1>, paramNames [Random Value, Speed, GlitchAmount, Chroma]}.GlitchAmount
  )
  ScalarParameter() {defaultValue=10, parameterName=GlitchAmount}
 ]
}
