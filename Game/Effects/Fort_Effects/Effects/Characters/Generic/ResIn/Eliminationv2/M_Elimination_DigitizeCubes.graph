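# Masked lit material for Niagara mesh particles, named for the digitize cubes in the character elimination effect.
# Particle alpha drives a lerp that offsets the second UV channel's V coordinate, and the clamped sum forms an alpha-controlled gradient wipe.
# A cyan Color parameter and two scalar constants sit alongside, with no material output links shown.
# Responsive AA is enabled.
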Material /Game/Effects/Fort_Effects/Effects/Characters/Generic/ResIn/Eliminationv2/M_Elimination_DigitizeCubes
graph {
 properties {
  bEnableResponsiveAA=true
  bUsedWithNiagaraMeshParticles=true
  blendMode=BLEND_Masked
  pixelDepthOffsetMode=PDOM_Legacy
  translucencyLightingMode=TLM_Surface
  usedShadingModels=Default Lit
 }
 nodes [
  n1 = TextureCoordinate() {coordinateIndex=1}
  n2 = ParticleColor()
  Clamp(None=Add(A=ComponentMask(None=@n1) {g=true}, B=LinearInterpolate(Alpha=@n2.A) {constA=-1}))
  Constant() {r=0.25}
  Constant() {r=0.75}
  VectorParameter() {parameterName=Color, defaultValue=<rgba 0.3167 0.9781 1 1>}
 ]
}
