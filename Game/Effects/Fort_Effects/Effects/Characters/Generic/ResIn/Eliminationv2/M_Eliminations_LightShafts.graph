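# Additive, unlit, two-sided Niagara particle material for light shafts in the character elimination effect.
# Three flare-texture samples at positive, zero and negative ChromaOffset form RGB fringing, with a dynamic parameter blending red- and green-channel masks for soft beams.
# Particle color tints the result, which passes through LDR-to-HDR remapping, desaturation and shared emissive adjustments.
# Particle alpha with depth fade, multiplied by camera-distance fade, softens intersections and close views.
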Material /Game/Effects/Fort_Effects/Effects/Characters/Generic/ResIn/Eliminationv2/M_Eliminations_LightShafts
graph {
 properties {
  bUsedWithNiagaraMeshParticles=true
  bUsedWithNiagaraSprites=true
  blendMode=BLEND_Additive
  pixelDepthOffsetMode=PDOM_Legacy
  shadingModel=MSM_Unlit
  twoSided=true
  usedShadingModels=Unlit
 }
 nodes [
  n1 = TextureCoordinate()
  n2 = ComponentMask(None=VectorParameter() {parameterName=ChromaOffset, defaultValue=<rgba 0.005 0 0 0>}.RGB) {g=true, r=true}
  n3 = TextureSample(UVs=Add(A=@n1, B=@n2)) {texture=/Game/Effects/Fort_Effects/Textures/Mask/T_Flare_LightShafts}
  n4 = TextureSample(UVs=@n1) {texture=&T_Flare_LightShafts}
  n5 = TextureSample(UVs=Add(A=@n1, B=Multiply(A=@n2) {constB=-1})) {texture=&T_Flare_LightShafts}
  n6 = ParticleColor()
  n7 = Multiply(
   A=LinearInterpolate(
    A=MaterialFunctionCall(X=@n3.R, Y=@n4.R, Z=@n5.R) {materialFunction=&MakeFloat3}.Result
    B=MaterialFunctionCall(X=@n3.G, Y=@n4.G, Z=@n5.G) {materialFunction=&MakeFloat3}.Result
    Alpha=DynamicParameter() {paramNames [Soft LightBeams, Param2, Param3, Param4]}.Soft LightBeams
   )
   B=@n6.RGB
  )
  n8 = Saturate(None=@n7)
  MaterialFunctionCall(Emissive=Desaturation(
    None=MaterialFunctionCall(LDR Input=@n8, HDR Tint=@n7) {materialFunction=&DeriveHDRfromLDR}.RemappedRange
    Fraction=ScalarParameter() {parameterName=Des}
   )) {materialFunction=&MF_FX_EmissiveAdjustments}
  Multiply(
   A=DepthFade(Opacity=@n6.A, FadeDistance=ScalarParameter() {defaultValue=100, parameterName=DepthFade})
   B=MaterialFunctionCall(
    Fade Length=ScalarParameter() {defaultValue=1000, parameterName=CameraFadeLength}
    Fade Offset=ScalarParameter() {parameterName=CameraFadeOffset}
   ) {materialFunction=&CameraDepthFade}.Result
  )
 ]
}
